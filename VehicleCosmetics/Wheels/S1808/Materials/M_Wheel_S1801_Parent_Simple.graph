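# Parent material for S1801 vehicle wheels, built on MF_WheelParent_Simple with optional tire shading and animated emissive.
# The Tire_Exists switch adds a colored Fresnel to base color, using tread-mirrored normals, optionally weighted by AO and limited by red vertex color.
# An FX mask texture drives time-based scrolling and sine-pulsing emissive.
# The result passes through the vehicle damage and de-rez transition functions.
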Material /VehicleCosmetics/Wheels/S1808/Materials/M_Wheel_S1801_Parent_Simple
graph {
 properties {
  bAutomaticallySetUsageInEditor=false
  bUseMaterialAttributes=true
  bUsedWithMeshDeformer=true
  bUsedWithSkeletalMesh=true
  pixelDepthOffsetMode=PDOM_Legacy
  usedShadingModels=Default Lit
 }
 nodes [
  Comment() {text=Tire Fresnel}
  Comment() {text=Tread Mirroring MF, commentColor=<rgba 0.3712 0.8796 1 1>}
  Comment() {text=Tire AO Bool}
  Composite() {subgraphName=Color Settings A-C}
  Composite() {subgraphName=Normal w/Tread Mirroring}
  Composite() {subgraphName=Default Settings}
  n1 = MaterialFunctionCall() {materialFunction=&MF_WheelParent_Simple}
  Composite!("In (MA)"=@n1.MaterialAttributes, TireAO=@n1.MaterialAttributes, TireVCMask=@n1.MaterialAttributes) {subgraphName=Tire Fresnel}
  MaterialExpressionReroute_4 = Reroute(None=@n1.MaterialAttributes)
  n2 = GetMaterialAttributes(None=@MaterialExpressionReroute_4."In (MA)") {attributeGetTypes [#1]}
  n3 = VectorParameter() {
   desc=Color fill for R value of RGB Mask
   group=3_Tire Properties
   parameterName=Tire_ColorA
   channelNames {a=Unused, b="", g="", r=""}
   defaultValue=<rgba 1 0.1248 0 1>
  }
  MaterialExpressionReroute_6 = Reroute(None=@n3)
  n4 = TextureSampleParameter2D(UVs=NamedRerouteUsage()) {group=1_Input Textures, parameterName=Tire_Normal, samplerType=SAMPLERTYPE_Normal, texture=/VehicleCosmetics/Textures/T_BlankNormal_B}
  n5 = MaterialFunctionCall(R=Multiply(A=@n4.R, B=NamedRerouteUsage()) {constB=-1}, G=Multiply(A=@n4.G, B=NamedRerouteUsage()) {constB=-1}, B=@n4.B) {materialFunction=&AppendMany}
  MaterialExpressionReroute_1 = Reroute(None=@n5.RGB)
  n6 = CameraVectorWS()
  n7 = Power(
   Base=DotProduct(
    A=@MaterialExpressionReroute_1.Tire_Normal
    B=Transform(None=@n6) {transformSourceType=TRANSFORMSOURCE_World, transformType=TRANSFORM_Tangent}
   )
   Exp=ScalarParameter() {defaultValue=6, group=5_Special Effects, parameterName=Tire_FresnelExponent, sortPriority=50}
  )
  n8 = Saturate(None=Multiply(
    A=ConstantBiasScale(None=@n7) {bias=0.0156}
    B=ScalarParameter() {defaultValue=1, group=5_Special Effects, parameterName=Tire_FresnelPower, sortPriority=50}
   ))
  MaterialExpressionReroute_15 = Reroute(None=@n1.TireRGBAMask)
  n9 = ComponentMask(None=@MaterialExpressionReroute_15.TireAO) {a=true, desc=Tire AO}
  n10 = ScalarParameter() {
   defaultValue=0.7
   desc=General Roughness that takes effect in absence of masking.
   group=4_Surface Defaults
   parameterName=Tire_RoughnessBase
   sliderMax=1
  }
  n11 = ScalarParameter() {desc=General Metalness that takes effect in absence of masking., group=4_Surface Defaults, parameterName=Tire_MetallicBase, sliderMax=1}
  n12 = ScalarParameter() {
   defaultValue=1
   desc="0==False, 1==True: AO gradient affects Emissive values. Use this for Emissives with depth. Values between 0 && 1 are allowed."
   group=5_Special Effects
   parameterName=Tire_AOTintsEmissive
  }
  n13 = ScalarParameter() {
   desc="0==False, 1==True: Blends AO with Diffuse. Values between 0 && 1 are allowed."
   group=5_Special Effects
   parameterName=Tire_AOTintsDiffuse
  }
  n14 = MaterialFunctionCall(R=@n10, G=@n11, B=@n12, A=@n13) {desc="Vector4 container for Rough, Metallic, AO settings", materialFunction=&AppendMany}
  MaterialExpressionReroute_2 = Reroute(None=@n14.RGBA)
  n15 = Saturate(None=ComponentMask(None=@MaterialExpressionReroute_2.RoughMetalAOTint) {a=true})
  n16 = LinearInterpolate(
   A=@n2.Base Color
   B=Multiply(
    A=ScalarParameter() {defaultValue=16, group=5_Special Effects, parameterName=Tire_FresnelBrightness, sortPriority=50}
    B=@MaterialExpressionReroute_6.ColorA
   )
   Alpha=LinearInterpolate(A=@n8, B=Multiply(A=@n9, B=@n8), Alpha=NamedRerouteDeclaration(None=@n15) {name=Bool_AO_Diffuse, nodeColor=<rgba 1 0.5859 0 1>})
  )
  MaterialExpressionReroute_14 = Reroute(None=@n1.VertexColorR)
  n17 = StaticSwitchParameter(
   True=BlendMaterialAttributes(
    A=@MaterialExpressionReroute_4."In (MA)"
    B=SetMaterialAttributes!(MaterialAttributes=@n2.MaterialAttributes, Base Color=@n16) {attributeSetTypes [#1]}
    Alpha=@MaterialExpressionReroute_14.TireVCMask
   )
   False=@MaterialExpressionReroute_4."In (MA)"
  ) {desc="Tire == True, No Tire == False. False is default.", group=5_Special Effects, parameterName=Tire_Exists}
  MaterialExpressionReroute_12 = Reroute(None=@n17)
  n18 = GetMaterialAttributes(None=@MaterialExpressionReroute_12.Output) {attributeGetTypes [#2]}
  n19 = TextureSampleParameter2D() {group=5_Special Effects, parameterName=FX Masks, sortPriority=0, texture=/VehicleCosmetics/Wheels/S1808/Textures/T_Wheel_S1808_Rim_Masks}
  n20 = ComponentMask(None=@n19.R) {r=true}
  n21 = Constant()
  n22 = Time()
  n23 = Multiply(A=@n22) {constB=0.5}
  n24 = Multiply(A=If(A=@n20, B=@n21, "A > B"=Constant() {r=1}, "A == B"=@n21, "A < B"=@n21), B=Frac(None=Subtract(A=@n19.G, B=@n23)))
  n25 = NamedRerouteUsage()
  n26 = NamedRerouteUsage()
  n27 = Multiply(A=@n25, B=OneMinus(None=@n26))
  n28 = Multiply(A=Clamp(None=ComponentMask(None=@n24) {r=true}), B=@n27)
  n29 = Multiply(A=Add(B=Sine(None=@n22) {period=2}) {constA=1}) {constB=0.5}
  n30 = Multiply(A=Multiply(A=@n20, B=@n29), B=OneMinus(None=@n19.G))
  n31 = Multiply(A=@n25, B=@n26)
  n32 = LinearInterpolate(A=Add(A=Add(A=@n28, B=@n27), B=Multiply(A=@n30) {constB=2}), B=Add(A=@n31, B=Multiply(A=@n31, B=@n29)), Alpha=NamedRerouteUsage())
  MaterialFunctionCall(In=MaterialFunctionCall(Material=SetMaterialAttributes!(MaterialAttributes=@n18.MaterialAttributes, Emissive Color=@n32) {attributeSetTypes [#2]}) {materialFunction=&MF_BaseVehicle_Damage}.Result) {materialFunction=&MF_VehicleDeRez_Transition}
  NamedRerouteDeclaration(None=@n1.VertexColorR) {name=VC_R, nodeColor=<rgba 1 0 0.3516 1>}
  NamedRerouteDeclaration(None=@n18.Emissive Color) {name=Emissive, nodeColor=<rgba 0 1 1 1>}
  n33 = MaterialFunctionCall() {
   desc="UV - For all Tread Mirrors. Lerp A && B - For Normals"
   materialFunction=/VehicleCosmetics/Wheels/_SharedMaterials/MF_Wheel_TreadMirroring
  }
  NamedRerouteDeclaration(None=@n33.LerpA) {name=Tread_LerpA, nodeColor=<rgba 0 1 0.6953 1>}
  NamedRerouteDeclaration(None=@n33.LerpB) {name=Tread_LerpB, nodeColor=<rgba 0 1 0.6953 1>}
  NamedRerouteDeclaration(None=@n33.UV) {name=Tread_UV, nodeColor=<rgba 0 1 0.6953 1>}
  n34 = VectorParameter() {
   desc="Emissive, Roughness, Metallic values for ColorA. Emissive has no limit, Roughness && Metallic are clamped between 0 && 1."
   group=3_Tire Properties
   parameterName=Tire_ColorA_PBR
   channelNames {a=Unused, b=Metallic, g=Roughness, r=Emissive}
   defaultValue=<rgba 0 0.7 0 1>
  }
  n35 = VectorParameter() {
   desc=Color fill for G value of RGB Mask
   group=3_Tire Properties
   parameterName=Tire_ColorB
   channelNames {a=Unused, b="", g="", r=""}
   defaultValue=<rgba 0 1 0.5029 1>
  }
  n36 = VectorParameter() {
   desc="Emissive, Roughness, Metallic values for ColorB. Emissive has no limit, Roughness && Metallic are clamped between 0 && 1."
   group=3_Tire Properties
   parameterName=Tire_ColorB_PBR
   channelNames {a=Unused, b=Metallic, g=Roughness, r=Emissive}
   defaultValue=<rgba 0 0.7 0 1>
  }
  n37 = VectorParameter() {
   desc=Color fill for B value of RGB Mask
   group=3_Tire Properties
   parameterName=Tire_ColorC
   channelNames {a=Unused, b="", g="", r=""}
   defaultValue=<rgba 0.5029 0 1 1>
  }
  n38 = VectorParameter() {
   desc="Emissive, Roughness, Metallic values for ColorC. Emissive has no limit, Roughness && Metallic are clamped between 0 && 1."
   group=3_Tire Properties
   parameterName=Tire_ColorC_PBR
   channelNames {a=Unused, b=Metallic, g=Roughness, r=Emissive}
   defaultValue=<rgba 0 0.7 0 1>
  }
  PinBase!(ColorA=@n3, ColorA_PBR=@n34, ColorB=@n35, ColorB_PBR=@n36, ColorC=@n37, ColorC_PBR=@n38) {reroutePins [
    ColorA {expression=&self:MaterialExpressionReroute_6}
    ColorA_PBR {expression=&self:MaterialExpressionReroute_7}
    ColorB {expression=&self:MaterialExpressionReroute_8}
    ColorB_PBR {expression=&self:MaterialExpressionReroute_9}
    ColorC {expression=&self:MaterialExpressionReroute_10}
    ColorC_PBR {expression=&self:MaterialExpressionReroute_11}
   ]}
  PinBase!()
  PinBase()
  PinBase!(Tire_Normal=@n5.RGB) {reroutePins [Tire_Normal {expression=&self:MaterialExpressionReroute_1}]}
  PinBase!()
  PinBase!(RoughMetalAOTint=@n14.RGBA) {reroutePins [RoughMetalAOTint {expression=&self:MaterialExpressionReroute_2}]}
  PinBase!() {reroutePins [
    "In (MA)" {expression=&self:MaterialExpressionReroute_4}
    TireAO {expression=&self:MaterialExpressionReroute_15}
    TireVCMask {expression=&self:MaterialExpressionReroute_14}
   ]}
  PinBase!(Output=@n17) {reroutePins [Output {expression=&self:MaterialExpressionReroute_12}]}
  MaterialExpressionReroute_10 = Reroute(None=@n37)
  MaterialExpressionReroute_11 = Reroute(None=@n38)
  MaterialExpressionReroute_7 = Reroute(None=@n34)
  MaterialExpressionReroute_8 = Reroute(None=@n35)
  MaterialExpressionReroute_9 = Reroute(None=@n36)
 ]
}
